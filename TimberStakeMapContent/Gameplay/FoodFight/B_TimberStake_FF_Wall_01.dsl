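; Food Fight wall static mesh actor that lights up when a button event with a matching ID is fired.
; On authority, it listens on the global GameplayEventRouter and sets FloorActivated, whose OnRep calls ActivateFloor to set LightAmount to full and print a debug string.
; Remotes zero LightAmount at BeginPlay, while the construction script sets it to full.
; Material writes cast slot 0 to a dynamic instance without creating one.
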(event EventBeginPlay
  (Networking|SwitchHasAuthority
    (:Authority
      (bind (_asynctaskproxy _payload) (GameplayEventRouter|StartListeningToEvent (GameplayEventRouter|GetGlobalEventRouter) "/TimberStakeMapContent/Gameplay/FoodFight/F_TimberStake_FoodFIght_ButtonPressedEvent.F_TimberStake_FoodFIght_ButtonPressedEvent")
        (:OnEventReceived
          (if (== (Utilities|Struct|BreakFTimberStakeFoodFIghtButtonPressedEvent _payload) (Variables|Default|GetButtonIDs))
            (Variables|Default|SetFloorActivated true)))))
    (:Remote
      (bind _asmaterial_instance_dynamic (Utilities|Casting|CastToMaterialInstanceDynamic (Rendering|Material|GetMaterial (Variables|StaticMeshActor|GetStaticMeshComponent) 0)))
      (Rendering|Material|SetScalarParameterValue _asmaterial_instance_dynamic "LightAmount" 0.0 :!class /Script/Engine.MaterialInstanceDynamic))))

(event Custom|ActivateFloor
  (bind _asmaterial_instance_dynamic (Utilities|Casting|CastToMaterialInstanceDynamic (Rendering|Material|GetMaterial (Variables|StaticMeshActor|GetStaticMeshComponent) 0)))
  (Rendering|Material|SetScalarParameterValue _asmaterial_instance_dynamic "LightAmount" 1.0 :!class /Script/Engine.MaterialInstanceDynamic)
  (Development|PrintString "ChangeMaterial"))

(fn OnRep_FloorActivated
  (if (Variables|Default|GetFloorActivated)
    (CallFunction|ActivateFloor)))

(fn ConstructionScript
  (bind _asmaterial_instance_dynamic (Utilities|Casting|CastToMaterialInstanceDynamic (Rendering|Material|GetMaterial (Variables|StaticMeshActor|GetStaticMeshComponent) 0)))
  (Rendering|Material|SetScalarParameterValue _asmaterial_instance_dynamic "LightAmount" 1.0 :!class /Script/Engine.MaterialInstanceDynamic))
