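; Interactable one-shot button for the TimberStake Food Fight mode that reports its press to the game.
; On BeginPlay it creates a dynamic material instance on its static mesh and sets the BaseColor parameter.
; Interacting sets HasBeenPressed, which blocks further interaction, and broadcasts a pressed event carrying the Button Id through the global GameplayEventRouter.
; The HasBeenPressed RepNotify only checks material validity and changes nothing.
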(event EventBeginPlay
  (bind _buttonmat (Variables|Default|GetButtonMat))
  (bind _returnvalue (Rendering|Material|CreateDynamicMaterialInstance (Variables|Debug|GetStaticMeshComponent) 0 (Rendering|Material|GetMaterial (Variables|Debug|GetStaticMeshComponent) 0) "None" :!class /Script/Engine.PrimitiveComponent))
  (Variables|Default|SetButtonMat _returnvalue)
  (Rendering|Material|SetMaterial (Variables|Debug|GetStaticMeshComponent) 0 _buttonmat)
  (Rendering|Material|SetVectorParameterValue _buttonmat "BaseColor" (Math|Color|SelectColor "(R=1.000000,G=0.000687,B=0.000000,A=1.000000)" "(R=0.005182,G=0.070360,B=0.021219,A=1.000000)") :!class /Script/Engine.MaterialInstanceDynamic))

(event Interaction|EventBlueprintOnInteract
  (params (InteractingPawn :type Fort%20Pawn%20Object%20Reference) (InteractionBeingAttempted :type EInteractionBeingAttempted%20Enum :default "FirstInteraction"))
  (|CallParentFunction InteractingPawn InteractionBeingAttempted)
  (Variables|Default|SetHasBeenPressed true)
  (GameplayEventRouter|BroadcastEvent (GameplayEventRouter|GetGlobalEventRouter) (Utilities|Struct|MakeFTimberStakeFoodFIghtButtonPressedEvent (Variables|Default|GetButton%20Id)) self))

(fn BlueprintCanInteract
  (params (InteractingPawn :type Fort%20Pawn%20Object%20Reference) (InteractionBeingAttempted :type EInteractionBeingAttempted%20Enum :default "FirstInteraction") (InteractionType :type TInteractionType%20Enum :default "IT_NoInteraction"))
  (return (not (Variables|Default|GetHasBeenPressed))))

(fn BlueprintGetInteractionString
  (params (InteractingPawn :type Fort%20Pawn%20Object%20Reference) (InteractionBeingAttempted :type EInteractionBeingAttempted%20Enum :default "FirstInteraction"))
  (return (Variables|Default|GetInteractionString)))

(fn OnRep_HasBeenPressed
  (Utilities|IsValid (Variables|Default|GetButtonMat)))

(fn ConstructionScript)
