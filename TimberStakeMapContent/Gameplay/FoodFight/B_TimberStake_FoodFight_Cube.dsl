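; Child Blueprint of a cube for the TimberStake Food Fight gameplay, adding control of a LightShape component that serves as a floor light.
; BeginPlay deactivates LightShape, then calls the parent.
; The overridden TurnOnFloorLight event calls the parent, activates LightShape without a reset and prints a "Hello" debug string.
; The construction script only defers to the parent.
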(event EventBeginPlay
  (Components|Activation|Deactivate (Variables|Default|GetLightShape))
  (|CallParentFunction))

(event EventTurnOnFloorLight
  (|CallParentFunction)
  (Components|Activation|Activate (Variables|Default|GetLightShape) false)
  (Development|PrintString "Hello"))

(fn ConstructionScript
  (|CallParentFunction))
