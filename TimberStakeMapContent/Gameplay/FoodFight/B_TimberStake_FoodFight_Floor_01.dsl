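; Food Fight floor piece that lights up when a button press event with a matching ID is received.
; The server listens on the global gameplay event router and sets FloorActivated when the event matches its ButtonIDs.
; FloorActivated's RepNotify raises LightAmount on a dynamic material created in the construction script and prints a debug string.
; Non-authority instances start with the light off; nothing here switches it off again after activation.
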(event EventBeginPlay
  (Networking|SwitchHasAuthority
    (:Authority
      (bind (_asynctaskproxy _payload) (GameplayEventRouter|StartListeningToEvent (GameplayEventRouter|GetGlobalEventRouter) "/TimberStakeMapContent/Gameplay/FoodFight/F_TimberStake_FoodFIght_ButtonPressedEvent.F_TimberStake_FoodFIght_ButtonPressedEvent")
        (:OnEventReceived
          (if (== (Utilities|Struct|BreakFTimberStakeFoodFIghtButtonPressedEvent _payload) (Variables|Default|GetButtonIDs))
            (Variables|Default|SetFloorActivated true)))))
    (:Remote
      (Rendering|Material|SetScalarParameterValue (Variables|Default|GetDynamicFloorMaterial) "LightAmount" 0.0 :!class /Script/Engine.MaterialInstanceDynamic))))

(event Custom|TurnOnFloorLight
  (Rendering|Material|SetScalarParameterValue (Variables|Default|GetDynamicFloorMaterial) "LightAmount" 1.0 :!class /Script/Engine.MaterialInstanceDynamic)
  (Development|PrintString "FloorChangeMaterial"))

(fn OnRep_FloorActivated
  (if (Variables|Default|GetFloorActivated)
    (CallFunction|TurnonFloorLight)))

(fn ConstructionScript
  (bind _returnvalue (Rendering|Material|CreateDynamicMaterialInstance (Variables|Default|GetSM_Blockout_Floor_A_A) 0 (Variables|Default|GetFloorMaterial) "None" :!class /Script/Engine.PrimitiveComponent))
  (Variables|Default|SetDynamicFloorMaterial _returnvalue))
