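; Golden llama loot actor for the TimberStake Food Fight mode that stays non-interactable until all its buttons are pressed.
; On authority, it listens for button-pressed router events and removes matching tags from ButtonIDs, enabling bCanInteract when none remain.
; The bCanInteract OnRep calls ReleaseGoldenLlama, which is empty here.
; PlayLootedFX runs a one-time cosmetic sequence: collision off, golden or standard effect, loop fade-out, loot sound and hiding.
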(event Collision|EventActorBeginOverlap
  (params (OtherActor :type Actor%20Object%20Reference)))

(event EventTick
  (params (DeltaSeconds :type Float%20%28single-precision%29)))

(event EventBeginPlay
  (Networking|SwitchHasAuthority
    (:Authority
      (bind (_asynctaskproxy _payload) (GameplayEventRouter|StartListeningToEvent (GameplayEventRouter|GetGlobalEventRouter) "/TimberStakeMapContent/Gameplay/FoodFight/F_TimberStake_FoodFIght_ButtonPressedEvent.F_TimberStake_FoodFIght_ButtonPressedEvent")
        (:OnEventReceived
          (if (GameplayTags|HasTag (Variables|Default|GetButtonIDs) (Utilities|Struct|BreakFTimberStakeFoodFIghtButtonPressedEvent _payload) true)
            (bind _returnvalue (GameplayTags|RemoveGameplayTag (Variables|Default|GetButtonIDs) (Utilities|Struct|BreakFTimberStakeFoodFIghtButtonPressedEvent _payload) :!class /Script/GameplayTags.BlueprintGameplayTagLibrary))
            (if _returnvalue
              (if (== (GameplayTags|GetNumGameplayTagsinContainer (Variables|Default|GetButtonIDs)) 0)
                (Variables|Default|SetbCanInteract true)))))))))

(event Custom|ReleaseGoldenLlama)

(fn BlueprintCanInteract
  (params (InteractingPawn :type Fort%20Pawn%20Object%20Reference) (InteractionBeingAttempted :type EInteractionBeingAttempted%20Enum :default "FirstInteraction") (InteractionType :type TInteractionType%20Enum :default "IT_NoInteraction"))
  (return (Variables|Default|GetbCanInteract)))

(fn OnRep_bCanInteract
  (CallFunction|ReleaseGoldenLlama))

(fn PlayLootedFX
  (bind _fxsocketname (Variables|Default|GetFXSocketName))
  (bind _llamastaticmesh (Variables|Default|GetLlamaStaticMesh))
  (bind _returnvalue (Transformation|GetSocketTransform _llamastaticmesh _fxsocketname "RTS_World"))
  (bind _location (.location _returnvalue))
  (bind _rotation (.rotation _returnvalue))
  (Networking|CanExecuteCosmeticEvents
    (:True
      (Utilities|FlowControl|DoOnce false
        (:Completed
          (Collision|SetActorEnableCollision false)
          (Collision|SetCollisionEnabled (Variables|Default|GetBoxCollision) "NoCollision")
          (if (Variables|Default|GetIsGoldenLlama)
            (Niagara|SpawnSystematLocation (Variables|Default|GetGoldenLlamaFX) _location _rotation)
            (else
              (Effects|SpawnEmitteratLocation (Variables|Looted|GetLlamaLootFX) _location _rotation)))
          (Audio|Components|Audio|FadeOut (Variables|Default|Gettreasure_pinata_loop_Cue) 0.25 0.0 "Linear" :!class /Script/Engine.AudioComponent)
          (Audio|SpawnSoundAttached (Variables|Looted|GetSound_Looted) (Variables|Default|GetLlamaStaticMesh) :!pinned true) ; Play sound for body sequence
          (Rendering|SetActorHiddenInGame true))))))

(fn ConstructionScript
  (|CallParentFunction))
