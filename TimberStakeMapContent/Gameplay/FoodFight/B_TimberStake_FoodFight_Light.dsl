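; Food Fight light actor for the TimberStake map, whose spotlight turns on when the button with a matching ID is pressed.
; The spotlight is hidden at BeginPlay, and the authority listens on the global event router for the button-pressed event.
; A matching event sets the replicated bHasBeenActivated, whose RepNotify calls OnButtonActivated to show the spotlight.
; Nothing in the graph turns the light off again.
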(event EventBeginPlay
  (Rendering|SetVisibility (Variables|Default|GetSpotLight) false false)
  (Networking|SwitchHasAuthority
    (:Authority
      (bind (_asynctaskproxy _payload) (GameplayEventRouter|StartListeningToEvent (GameplayEventRouter|GetGlobalEventRouter) "/TimberStakeMapContent/Gameplay/FoodFight/F_TimberStake_FoodFIght_ButtonPressedEvent.F_TimberStake_FoodFIght_ButtonPressedEvent")
        (:OnEventReceived
          (if (== (Utilities|Struct|BreakFTimberStakeFoodFIghtButtonPressedEvent _payload) (Variables|Default|GetButtonId))
            (Variables|Default|SetbHasBeenActivated true)))))))

(event Custom|OnButtonActivated
  ; Add logic to turn the light on here
  (Rendering|SetVisibility (Variables|Default|GetSpotLight) true false))

(fn OnRep_bHasBeenActivated
  (CallFunction|OnButtonActivated))

(fn ConstructionScript)
