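# Nanite-enabled lit material for the FoodFight cube in TimberStake, with selectable billboard wall textures per face and a glowing shape.
# Local-space normals, sharpened into directional weights, blend top, front, back, left and right textures chosen by index parameters.
# The Y weight input reuses local X, so side weights mirror front and back.
# An offsettable, scalable, box-masked shape texture drives adjustable emissive on positive-facing sides; shared billboard normal and mask textures are sampled.
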Material /TimberStakeMapContent/Gameplay/FoodFight/M_TimberStake_FoodFight_Cube
graph {
 properties {bAutomaticallySetUsageInEditor=false, bUsedWithNanite=true, usedShadingModels=Default Lit}
 nodes [
  n1 = ScalarParameter() {defaultValue=1, group=Cube, parameterName=TopIndex, sliderMax=3, sortPriority=10}
  n2 = TextureSample() {
   samplerSource=SSM_Wrap_WorldGroupSettings
   texture=/TimberStakeMapContent/Environments/Sets/Billboard/Textures/T_Billboard_Exterior_Wall_04_D
  }
  n3 = TextureSample() {
   samplerSource=SSM_Wrap_WorldGroupSettings
   texture=/TimberStakeMapContent/Environments/Sets/Billboard/Textures/T_Billboard_Exterior_Wall_03_D
  }
  n4 = TextureSample() {
   samplerSource=SSM_Wrap_WorldGroupSettings
   texture=/TimberStakeMapContent/Environments/Sets/Billboard/Textures/T_Billboard_Exterior_Wall_02_D
  }
  n5 = TextureSample() {
   samplerSource=SSM_Wrap_WorldGroupSettings
   texture=/TimberStakeMapContent/Environments/Sets/Billboard/Textures/T_Billboard_Exterior_Wall_01_D
  }
  n6 = If(A=@n1, "A > B"=@n3.RGB, "A < B"=If(A=@n1, "A > B"=@n4.RGB, "A < B"=@n5.RGB) {constB=0.5}) {constB=1.5}
  n7 = MaterialFunctionCall(Float3=Transform(None=VertexNormalWS()) {transformSourceType=TRANSFORMSOURCE_World, transformType=TRANSFORM_Local}) {materialFunction=&BreakOutFloat3Components}
  n8 = MaterialFunctionCall(X=@n7.R, Y=@n7.R, Z=Saturate(None=@n7.B)) {materialFunction=&MakeFloat3}
  n9 = Power(Base=Abs(None=@n8.Result)) {constExponent=16}
  n10 = Constant3Vector() {constant=<rgba 1 1 1 0>}
  n11 = DotProduct(A=@n9, B=@n10)
  n12 = Multiply(A=Sign(None=@n8.Result), B=Divide(A=@n9, B=Max(A=@n11) {constB=0.0001}))
  n13 = Saturate(None=@n12)
  n14 = MaterialFunctionCall(Float3=@n13) {materialFunction=&BreakOutFloat3Components}
  n15 = ScalarParameter() {group=Cube, parameterName=FrontIndex, sliderMax=3, sortPriority=11}
  n16 = If(A=@n15, "A > B"=@n3.RGB, "A < B"=If(A=@n15, "A > B"=@n4.RGB, "A < B"=@n5.RGB) {constB=0.5}) {constB=1.5}
  n17 = Add(
   A=Multiply(A=If(A=@n1, "A > B"=@n2.RGB, "A < B"=@n6) {constB=2.5}, B=@n14.B)
   B=Multiply(A=If(A=@n15, "A > B"=@n2.RGB, "A < B"=@n16) {constB=2.5}, B=@n14.R)
  )
  n18 = ScalarParameter() {defaultValue=3, group=Cube, parameterName=BackIndex, sliderMax=3, sortPriority=13}
  n19 = If(A=@n18, "A > B"=@n4.RGB, "A < B"=@n5.RGB) {constB=0.5}
  n20 = MaterialFunctionCall(Float3=Saturate(None=Multiply(A=@n12) {constB=-1})) {materialFunction=&BreakOutFloat3Components}
  n21 = Multiply(A=If(A=@n18, "A > B"=@n2.RGB, "A < B"=If(A=@n18, "A > B"=@n3.RGB, "A < B"=@n19) {constB=1.5}) {constB=2.5}, B=@n20.R)
  n22 = ScalarParameter() {defaultValue=3, group=Cube, parameterName=LeftIndex, sliderMax=3, sortPriority=14}
  n23 = If(A=@n22, "A > B"=@n2.RGB, "A < B"=If(A=@n22, "A > B"=@n3.RGB, "A < B"=If(A=@n22, "A > B"=@n4.RGB, "A < B"=@n5.RGB) {constB=0.5}) {constB=1.5}) {constB=2.5}
  n24 = ScalarParameter() {defaultValue=2, group=Cube, parameterName=RightIndex, sliderMax=3, sortPriority=12}
  n25 = If(A=@n24, "A > B"=@n3.RGB, "A < B"=If(A=@n24, "A > B"=@n4.RGB, "A < B"=@n5.RGB) {constB=0.5}) {constB=1.5}
  Add(A=Add(A=Add(A=@n17, B=@n21), B=Multiply(A=@n23, B=@n20.G)), B=Multiply(A=If(A=@n24, "A > B"=@n2.RGB, "A < B"=@n25) {constB=2.5}, B=@n14.G))
  n26 = VectorParameter() {group=Cube, parameterName=EmissiveMultipler, defaultValue=<rgba 5 5 5 0>}
  n27 = Frac(None=TextureCoordinate())
  n28 = Add(
   A=Multiply(A=Subtract(A=@n27) {constB=0.5}, B=ScalarParameter() {defaultValue=1.5, group=Cube, parameterName=ShapeScale})
   B=ComponentMask(None=VectorParameter() {group=Cube, parameterName=ShapeUVOffset}.RGB) {g=true, r=true}
  )
  n29 = TextureSampleParameter2D(UVs=Add(A=@n28) {constB=0.5}) {
   group=Cube
   parameterName=Shape
   samplerSource=SSM_Clamp_WorldGroupSettings
   texture=&/Cube_Pattern_Umbrella
  }
  n30 = Constant2Vector() {g=1, r=1}
  n31 = Constant() {r=0.01}
  n32 = DotProduct(A=@n13, B=@n10)
  n33 = Multiply(
   A=Multiply(
    A=@n29.RGB
    B=MaterialFunctionCall!(A=@n28, B, Bounds=@n30, Edge Falloff=@n31) {materialFunction=/Engine/Functions/Engine_MaterialFunctions01/Gradient/BoxMask-2D}.Result
   )
   B=Saturate(None=Multiply(A=@n32) {constB=10000})
  )
  Multiply(
   A=MaterialFunctionCall(Emissive=Multiply(A=@n26.RGB, B=@n33)) {materialFunction=&MF_FX_EmissiveAdjustments}.Result
   B=ScalarParameter() {defaultValue=1, group=Cube, parameterName=LightAmount}
  )
  TextureSample() {samplerType=SAMPLERTYPE_Normal, texture=/TimberStakeMapContent/Environments/Sets/Billboard/Textures/T_Billboard_Exterior_N}
  TextureSample() {samplerType=SAMPLERTYPE_Masks, texture=/TimberStakeMapContent/Environments/Sets/Billboard/Textures/T_Billboard_Exterior_S}
 ]
}
