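# Default Lit material for the TimberStake Food Fight wall, pairing billboard diffuse, specular mask and normal textures with an emissive symbol, by default an umbrella pattern.
# The symbol samples UV0 scaled uniformly about its centre by ShapeSize's red channel, with a BoxMask-2D trimming anything outside the square.
# EmissiveMult tints it, and an If shows it only where UV1's U is below 1.
# LightAmount scales the result before MF_FX_EmissiveAdjustments.
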Material /TimberStakeMapContent/Gameplay/FoodFight/M_TimberStake_FoodFight_Wall
graph {
 properties {bAutomaticallySetUsageInEditor=false, usedShadingModels=Default Lit}
 nodes [
  n1 = ComponentMask(None=TextureCoordinate() {coordinateIndex=1}) {r=true}
  n2 = Constant()
  n3 = VectorParameter() {parameterName=ShapeSize, defaultValue=<rgba 1.75 1.3125 0 0>}
  n4 = Multiply(
   A=Subtract(A=TextureCoordinate()) {constB=0.5}
   B=MaterialFunctionCall(X=@n3.R, Y=@n3.R) {materialFunction=&MakeFloat2}.Result
  )
  n5 = TextureSampleParameter2D(UVs=Add(A=@n4) {constB=0.5}) {
   parameterName=Diffuse_Texture_2
   samplerSource=SSM_Clamp_WorldGroupSettings
   texture=&/Cube_Pattern_Umbrella
  }
  n6 = Constant2Vector() {g=1, r=1}
  n7 = Constant() {r=0.01}
  n8 = Multiply(
   A=VectorParameter() {parameterName=EmissiveMult, defaultValue=<rgba 0.5 0.5 0.5 0>}.RGB
   B=Multiply(
    A=@n5.RGB
    B=MaterialFunctionCall!(A=@n4, B, Bounds=@n6, Edge Falloff=@n7) {materialFunction=/Engine/Functions/Engine_MaterialFunctions01/Gradient/BoxMask-2D}.Result
   )
  )
  MaterialFunctionCall(Emissive=Multiply(A=If(A=@n1, "A > B"=@n2, "A == B"=@n2, "A < B"=@n8) {constB=1}, B=ScalarParameter() {defaultValue=1, parameterName=LightAmount, sliderMax=1})) {materialFunction=&MF_FX_EmissiveAdjustments}
  TextureSampleParameter2D() {parameterName=Diffuse, texture=/TimberStakeMapContent/Environments/Sets/Billboard/Textures/T_Billboard_Exterior_Wall_01_D}
  TextureSampleParameter2D() {
   parameterName=SpecularMasks
   samplerType=SAMPLERTYPE_Masks
   texture=/TimberStakeMapContent/Environments/Sets/Billboard/Textures/T_Billboard_Exterior_S
  }
  TextureSampleParameter2D() {
   parameterName=Normals
   samplerType=SAMPLERTYPE_Normal
   texture=/TimberStakeMapContent/Environments/Sets/Billboard/Textures/T_Billboard_Exterior_N
  }
 ]
}
